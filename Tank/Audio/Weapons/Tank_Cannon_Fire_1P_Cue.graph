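# Sound cue for the first-person tank cannon shot, pairing a close fire sound with a reload.
# A quality-level switch picks a random pool of four fire waves for its first two levels or two for the others, then feeds a fixed-pitch modulator.
# A mixer layers that with a quieter reload wave delayed one second.
# First-person attenuation, sound class and gunfire concurrency apply.
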SoundCue /Tank/Audio/Weapons/Tank_Cannon_Fire_1P_Cue
root=Mixer_0
nodes [
 {id=Delay_0, children [WavePlayer_1], properties {delayMax=1, delayMin=1}}
 {id=Mixer_0, children [Modulator_6,Delay_0], properties {inputVolume [1,0.35]}}
 {id=Modulator_6, children [QualityLevel_0], properties {pitchMax=1, pitchMin=1}}
 {id=QualityLevel_0, children [Random_1,Random_1,Random_2,Random_2]}
 {
  id=Random_1
  children [WavePlayer_0,WavePlayer_4,WavePlayer_5,WavePlayer_6]
  properties {bShouldExcludeFromBranchCulling=true, weights [1,1,1,1]}
 }
 {id=Random_2, children [WavePlayer_2,WavePlayer_3], properties {weights [1,1]}}
 {id=WavePlayer_0, properties {soundWaveAssetPtr=&/WPN_Tank_Cannon_Fire_Close_01}}
 {id=WavePlayer_1, properties {soundWaveAssetPtr=&/WPN_Tank_Cannon_Reload_01}}
 {id=WavePlayer_2, properties {soundWaveAssetPtr=&WPN_Tank_Cannon_Fire_Close_01}}
 {id=WavePlayer_3, properties {soundWaveAssetPtr=&/WPN_Tank_Cannon_Fire_Close_02}}
 {id=WavePlayer_4, properties {soundWaveAssetPtr=&WPN_Tank_Cannon_Fire_Close_02}}
 {id=WavePlayer_5, properties {soundWaveAssetPtr=&/WPN_Tank_Cannon_Fire_Close_03}}
 {id=WavePlayer_6, properties {soundWaveAssetPtr=&/WPN_Tank_Cannon_Fire_Close_04}}
]
properties {
 attenuationSettings=&Fort_Weapon_Projectile_FirstPerson_Athena_Att
 maxDistance=2097152
 soundClassObject=&Fort_WeaponRanged_Fire_1P_PSM
 volumeMultiplier=1.5
 concurrencyOverrides {maxCount=4}
 concurrencySet [&Weapon_Gunfire_FirstPerson_Conc]
}
