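# Sound cue for the tank cannon's firing sound, layered by listener distance.
# Randomized close, medium and distant shots are mixed through separate attenuations and sound class overrides, with a quieter reload sound delayed one second in the close layer.
# A quality-level switch gives the first two levels the full variation pools and the last two smaller pools.
# Concurrency override limits simultaneous instances to three.
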SoundCue /Tank/Audio/Weapons/Tank_Cannon_Fire_Cue
root=QualityLevel_2
nodes [
 {id=Attenuation_10, children [Random_5], properties {attenuationSettings=/Game/Sounds/Attenuation/Fort_Weapon_Projectile_Medium_StW_Att}}
 {
  id=Attenuation_11
  children [Modulator_7]
  properties {attenuationSettings=&Fort_Weapon_Projectile_Distant_Mono_Athena_Att}
 }
 {id=Attenuation_3, children [Mixer_8], properties {attenuationSettings=&Fort_Weapon_Projectile_Close_Athena_Att}}
 {id=Attenuation_4, children [Random_7], properties {attenuationSettings=&Fort_Weapon_Projectile_Medium_StW_Att}}
 {
  id=Attenuation_5
  children [Modulator_5]
  properties {attenuationSettings=&Fort_Weapon_Projectile_Distant_Mono_Athena_Att}
 }
 {id=Attenuation_9, children [Mixer_1], properties {attenuationSettings=&Fort_Weapon_Projectile_Close_Athena_Att}}
 {id=Delay_0, children [WavePlayer_0], properties {delayMax=1, delayMin=1}}
 {id=Delay_2, children [WavePlayer_41], properties {delayMax=1, delayMin=1}}
 {id=Mixer_0, children [Modulator_1,SoundClass_1], properties {inputVolume [1,1]}}
 {id=Mixer_1, children [Modulator_6,Delay_0], properties {inputVolume [1,0.35]}}
 {id=Mixer_3, children [SoundClass_7,Mixer_0], properties {inputVolume [1,2]}}
 {id=Mixer_6, children [SoundClass_5,Mixer_7], properties {inputVolume [1,2]}}
 {id=Mixer_7, children [Modulator_8,SoundClass_8], properties {inputVolume [1,1]}}
 {id=Mixer_8, children [Modulator_4,Delay_2], properties {inputVolume [1,0.35]}}
 {id=Modulator_1, children [SoundClass_0], properties {volumeMax=1, volumeMin=1}}
 {id=Modulator_4, children [Random_3], properties {volumeMax=1.25, volumeMin=1.2}}
 {id=Modulator_5, children [Random_8], properties {pitchMax=1.1}}
 {id=Modulator_6, children [Random_4], properties {volumeMax=1.25, volumeMin=1.2}}
 {id=Modulator_7, children [Random_6], properties {pitchMax=1.1}}
 {id=Modulator_8, children [SoundClass_6], properties {volumeMax=1, volumeMin=1}}
 {id=QualityLevel_2, children [Mixer_3,Mixer_3,Mixer_6,Mixer_6]}
 {id=Random_3, children [WavePlayer_15, WavePlayer_16, WavePlayer_30], properties {bShouldExcludeFromBranchCulling=true, weights [1,1,1]}}
 {
  id=Random_4
  children [WavePlayer_17, WavePlayer_18, WavePlayer_19, WavePlayer_20]
  properties {bShouldExcludeFromBranchCulling=true, weights [1,1,1,1]}
 }
 {
  id=Random_5
  children [WavePlayer_21, WavePlayer_22, WavePlayer_23, WavePlayer_24]
  properties {bShouldExcludeFromBranchCulling=true, weights [1,1,1,1]}
 }
 {id=Random_6, children [WavePlayer_25, WavePlayer_26, WavePlayer_27, WavePlayer_28, WavePlayer_29], properties {weights [1,1,1,1,1]}}
 {id=Random_7, children [WavePlayer_32, WavePlayer_33], properties {bShouldExcludeFromBranchCulling=true, weights [1,1]}}
 {id=Random_8, children [WavePlayer_36, WavePlayer_37], properties {weights [1,1]}}
 {
  id=SoundClass_0
  children [Attenuation_10]
  properties {soundClassOverride=&Fort_WeaponRanged_Fire_Distant}
 }
 {
  id=SoundClass_1
  children [Attenuation_11]
  properties {soundClassOverride=&Fort_WeaponRanged_Fire_Distant}
 }
 {
  id=SoundClass_5
  children [Attenuation_3]
  properties {soundClassOverride=&Fort_WeaponRanged_Fire_3P_PSM}
 }
 {
  id=SoundClass_6
  children [Attenuation_4]
  properties {soundClassOverride=&Fort_WeaponRanged_Fire_Distant}
 }
 {
  id=SoundClass_7
  children [Attenuation_9]
  properties {soundClassOverride=&Fort_WeaponRanged_Fire_3P_PSM}
 }
 {
  id=SoundClass_8
  children [Attenuation_5]
  properties {soundClassOverride=&Fort_WeaponRanged_Fire_Distant}
 }
 {id=WavePlayer_0, properties {soundWaveAssetPtr=&/WPN_Tank_Cannon_Reload_01}}
 {id=WavePlayer_15, properties {soundWaveAssetPtr=&/WPN_Tank_Cannon_Fire_Close_01}}
 {id=WavePlayer_16, properties {soundWaveAssetPtr=&/WPN_Tank_Cannon_Fire_Close_02}}
 {id=WavePlayer_17, properties {soundWaveAssetPtr=&WPN_Tank_Cannon_Fire_Close_01}}
 {id=WavePlayer_18, properties {soundWaveAssetPtr=&WPN_Tank_Cannon_Fire_Close_02}}
 {id=WavePlayer_19, properties {soundWaveAssetPtr=&/WPN_Tank_Cannon_Fire_Close_03}}
 {id=WavePlayer_20, properties {soundWaveAssetPtr=&/WPN_Tank_Cannon_Fire_Close_04}}
 {id=WavePlayer_21, properties {soundWaveAssetPtr=&/WPN_Tank_Cannon_Fire_Med_01}}
 {id=WavePlayer_22, properties {soundWaveAssetPtr=&/WPN_Tank_Cannon_Fire_Med_02}}
 {id=WavePlayer_23, properties {soundWaveAssetPtr=&/WPN_Tank_Cannon_Fire_Med_03}}
 {id=WavePlayer_24, properties {soundWaveAssetPtr=&/WPN_Tank_Cannon_Fire_Med_04}}
 {id=WavePlayer_25, properties {soundWaveAssetPtr=&/WPN_Tank_Cannon_Fire_Dist_01}}
 {id=WavePlayer_26, properties {soundWaveAssetPtr=&/WPN_Tank_Cannon_Fire_Dist_02}}
 {id=WavePlayer_27, properties {soundWaveAssetPtr=&/WPN_Tank_Cannon_Fire_Dist_03}}
 {id=WavePlayer_28, properties {soundWaveAssetPtr=&/WPN_Tank_Cannon_Fire_Dist_04}}
 {id=WavePlayer_29, properties {soundWaveAssetPtr=&/WPN_Tank_Cannon_Fire_Dist_05}}
 {id=WavePlayer_30, properties {soundWaveAssetPtr=&WPN_Tank_Cannon_Fire_Close_03}}
 {id=WavePlayer_32, properties {soundWaveAssetPtr=&WPN_Tank_Cannon_Fire_Med_01}}
 {id=WavePlayer_33, properties {soundWaveAssetPtr=&WPN_Tank_Cannon_Fire_Med_02}}
 {id=WavePlayer_36, properties {soundWaveAssetPtr=&WPN_Tank_Cannon_Fire_Dist_01}}
 {id=WavePlayer_37, properties {soundWaveAssetPtr=&WPN_Tank_Cannon_Fire_Dist_02}}
 {id=WavePlayer_41, properties {soundWaveAssetPtr=&WPN_Tank_Cannon_Reload_01}}
]
properties {
 bOverrideConcurrency=true
 maxDistance=27500
 volumeMultiplier=1
 attenuationOverrides {falloffDistance=20000}
 concurrencyOverrides {maxCount=3}
}
