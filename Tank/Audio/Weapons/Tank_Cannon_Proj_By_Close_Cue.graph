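# Sound cue for a tank cannon projectile passing close by, classed as a ranged-weapon projectile effect.
# A quality switch sends the first two levels to a random pick of four waves and the other two to three similarly named waves.
# A modulator randomizes pitch upward for variety.
# It references shared bullet-whip attenuation and limits overlapping plays through bullet whiz-by concurrency.
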SoundCue /Tank/Audio/Weapons/Tank_Cannon_Proj_By_Close_Cue
root=Modulator_0
nodes [
 {id=Modulator_0, children [QualityLevel_0], properties {pitchMax=1.25, pitchMin=1}}
 {id=QualityLevel_0, children [Random_0,Random_0,Random_1,Random_1]}
 {
  id=Random_0
  children [WavePlayer_0,WavePlayer_1,WavePlayer_2,WavePlayer_3]
  properties {bShouldExcludeFromBranchCulling=true, weights [1,1,1,1]}
 }
 {id=Random_1, children [WavePlayer_4,WavePlayer_5,WavePlayer_6], properties {bShouldExcludeFromBranchCulling=true, weights [1,1,1]}}
 {id=WavePlayer_0, properties {soundWaveAssetPtr=&/WPN_Tank_Cannon_Proj_By_Close_04}}
 {id=WavePlayer_1, properties {soundWaveAssetPtr=&/WPN_Tank_Cannon_Proj_By_Close_01}}
 {id=WavePlayer_2, properties {soundWaveAssetPtr=&/WPN_Tank_Cannon_Proj_By_Close_02}}
 {id=WavePlayer_3, properties {soundWaveAssetPtr=&/WPN_Tank_Cannon_Proj_By_Close_03}}
 {id=WavePlayer_4, properties {soundWaveAssetPtr=&WPN_Tank_Cannon_Proj_By_Close_04}}
 {id=WavePlayer_5, properties {soundWaveAssetPtr=&WPN_Tank_Cannon_Proj_By_Close_01}}
 {id=WavePlayer_6, properties {soundWaveAssetPtr=&WPN_Tank_Cannon_Proj_By_Close_02}}
]
properties {
 attenuationSettings=/Game/Athena/Sounds/Weapons/BulletWhips/bullet_whip_Att.Bullet_Whip_Att
 maxDistance=2097152
 soundClassObject=&Fort_WeaponRanged_Projectile_Effects
 volumeMultiplier=1.2
 attenuationOverrides {bAttenuate=false, stereoSpread=0}
 concurrencySet [/Game/Sounds/Concurrency_Rules/Athena_BulletWhizBy_Conc]
}
