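# Sound cue for the explosion of a tank cannon projectile, layered by listener distance.
# A quality node feeds the first two levels a mixer with four variants per layer, and the last two a mixer with fewer.
# Each mixer blends close, medium and distant random wave picks, each with its own attenuation and sound class.
# Medium and distant layers get pitch variation and boosted volume, and overridden concurrency caps simultaneous instances.
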SoundCue /Tank/Audio/Weapons/Tank_Cannon_Proj_Expl_Cue
root=QualityLevel_0
nodes [
 {
  id=Attenuation_0
  children [Modulator_1]
  properties {
   attenuationSettings=/Game/Sounds/Attenuation/Fort_Explosion_Distant_Athena_Att
   attenuationOverrides {falloffDistance=3000, attenuationShapeExtents {x=1000}}
  }
 }
 {
  id=Attenuation_4
  children [Random_6]
  properties {attenuationSettings=/Game/Sounds/Attenuation/Fort_Explosion_Close_Athena_Att, attenuationOverrides {falloffDistance=1600}}
 }
 {
  id=Attenuation_5
  children [Modulator_5]
  properties {
   attenuationSettings=/Game/Sounds/Attenuation/Fort_Explosion_Medium_Athena_Att
   attenuationOverrides {falloffDistance=3000, attenuationShapeExtents {x=1000}}
  }
 }
 {
  id=Attenuation_6
  children [Random_3]
  properties {attenuationSettings=&Fort_Explosion_Close_Athena_Att, attenuationOverrides {falloffDistance=1600}}
 }
 {
  id=Attenuation_7
  children [Modulator_6]
  properties {
   attenuationSettings=&Fort_Explosion_Medium_Athena_Att
   attenuationOverrides {falloffDistance=3000, attenuationShapeExtents {x=1000}}
  }
 }
 {
  id=Attenuation_8
  children [Modulator_8]
  properties {
   attenuationSettings=&Fort_Explosion_Distant_Athena_Att
   attenuationOverrides {falloffDistance=3000, attenuationShapeExtents {x=1000}}
  }
 }
 {id=Mixer_1, children [SoundClass_7,SoundClass_6,SoundClass_8], properties {inputVolume [1,2,2]}}
 {id=Mixer_4, children [SoundClass_3,SoundClass_2,SoundClass_0], properties {inputVolume [1,2,2]}}
 {id=Modulator_0, children [Attenuation_4]}
 {id=Modulator_1, children [Random_8], properties {pitchMax=1.1, pitchMin=0.9}}
 {id=Modulator_5, children [Random_7], properties {pitchMax=1.1, pitchMin=0.9}}
 {id=Modulator_6, children [Random_4], properties {pitchMax=1.1, pitchMin=0.9}}
 {id=Modulator_7, children [Attenuation_6]}
 {id=Modulator_8, children [Random_5], properties {pitchMax=1.1, pitchMin=0.9}}
 {id=QualityLevel_0, children [Mixer_4,Mixer_4,Mixer_1,Mixer_1]}
 {id=Random_3, children [WavePlayer_16, WavePlayer_17, WavePlayer_18], properties {bShouldExcludeFromBranchCulling=true, weights [1,1,1]}}
 {id=Random_4, children [WavePlayer_20, WavePlayer_21, WavePlayer_22], properties {bShouldExcludeFromBranchCulling=true, weights [1,1,1]}}
 {id=Random_5, children [WavePlayer_24, WavePlayer_25], properties {weights [1,1]}}
 {
  id=Random_6
  children [WavePlayer_7, WavePlayer_8, WavePlayer_13, WavePlayer_14]
  properties {bShouldExcludeFromBranchCulling=true, weights [1,1,1,1]}
 }
 {
  id=Random_7
  children [WavePlayer_33, WavePlayer_34, WavePlayer_35, WavePlayer_36]
  properties {bShouldExcludeFromBranchCulling=true, weights [1,1,1,1]}
 }
 {id=Random_8, children [WavePlayer_37, WavePlayer_38, WavePlayer_39, WavePlayer_40], properties {weights [1,1,1,1]}}
 {
  id=SoundClass_0
  children [Attenuation_0]
  properties {soundClassOverride=&Fort_WeaponRanged_Fire_Distant}
 }
 {
  id=SoundClass_2
  children [Attenuation_5]
  properties {soundClassOverride=&Fort_WeaponRanged_Fire_Distant}
 }
 {id=SoundClass_3, children [Modulator_0], properties {soundClassOverride=&Fort_WeaponRanged_Fire_3P_PSM}}
 {
  id=SoundClass_6
  children [Attenuation_7]
  properties {soundClassOverride=&Fort_WeaponRanged_Fire_Distant}
 }
 {id=SoundClass_7, children [Modulator_7], properties {soundClassOverride=&Fort_WeaponRanged_Fire_3P_PSM}}
 {
  id=SoundClass_8
  children [Attenuation_8]
  properties {soundClassOverride=&Fort_WeaponRanged_Fire_Distant}
 }
 {id=WavePlayer_13, properties {soundWaveAssetPtr=&/WPN_Tank_Cannon_Proj_Expl_Close_03}}
 {id=WavePlayer_14, properties {soundWaveAssetPtr=&/WPN_Tank_Cannon_Proj_Expl_Close_04}}
 {id=WavePlayer_16, properties {soundWaveAssetPtr=&/WPN_Tank_Cannon_Proj_Expl_Close_01}}
 {id=WavePlayer_17, properties {soundWaveAssetPtr=&/WPN_Tank_Cannon_Proj_Expl_Close_02}}
 {id=WavePlayer_18, properties {soundWaveAssetPtr=&WPN_Tank_Cannon_Proj_Expl_Close_03}}
 {id=WavePlayer_20, properties {soundWaveAssetPtr=&/WPN_Tank_Cannon_Proj_Expl_Med_01}}
 {id=WavePlayer_21, properties {soundWaveAssetPtr=&/WPN_Tank_Cannon_Proj_Expl_Med_02}}
 {id=WavePlayer_22, properties {soundWaveAssetPtr=&/WPN_Tank_Cannon_Proj_Expl_Med_03}}
 {id=WavePlayer_24, properties {soundWaveAssetPtr=&/WPN_Tank_Cannon_Proj_Expl_Dist_04}}
 {id=WavePlayer_25, properties {soundWaveAssetPtr=&/WPN_Tank_Cannon_Proj_Expl_Dist_01}}
 {id=WavePlayer_33, properties {soundWaveAssetPtr=&WPN_Tank_Cannon_Proj_Expl_Med_01}}
 {id=WavePlayer_34, properties {soundWaveAssetPtr=&WPN_Tank_Cannon_Proj_Expl_Med_02}}
 {id=WavePlayer_35, properties {soundWaveAssetPtr=&WPN_Tank_Cannon_Proj_Expl_Med_03}}
 {id=WavePlayer_36, properties {soundWaveAssetPtr=&/WPN_Tank_Cannon_Proj_Expl_Med_04}}
 {id=WavePlayer_37, properties {soundWaveAssetPtr=&WPN_Tank_Cannon_Proj_Expl_Dist_04}}
 {id=WavePlayer_38, properties {soundWaveAssetPtr=&WPN_Tank_Cannon_Proj_Expl_Dist_01}}
 {id=WavePlayer_39, properties {soundWaveAssetPtr=&/WPN_Tank_Cannon_Proj_Expl_Dist_02}}
 {id=WavePlayer_40, properties {soundWaveAssetPtr=&/WPN_Tank_Cannon_Proj_Expl_Dist_03}}
 {id=WavePlayer_7, properties {soundWaveAssetPtr=&WPN_Tank_Cannon_Proj_Expl_Close_01}}
 {id=WavePlayer_8, properties {soundWaveAssetPtr=&WPN_Tank_Cannon_Proj_Expl_Close_02}}
]
properties {
 bOverrideConcurrency=true
 maxDistance=27500
 volumeMultiplier=1
 concurrencyOverrides {maxCount=3}
 concurrencySet [/Game/Sounds/Concurrency_Rules/Weapon_Explosion_Conc]
}
