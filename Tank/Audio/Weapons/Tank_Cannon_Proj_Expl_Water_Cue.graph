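# Sound cue for a tank cannon projectile exploding on water, layered by listener distance.
# Each branch mixes close explosion plus water debris, medium and distant layers, each with random wave selection, pitch modulation and its own attenuation.
# Sound class overrides separate the close layer from the medium and distant ones.
# A quality node feeds the first two levels a fuller variation set and the last two a trimmed one, with overridden concurrency limiting instances.
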SoundCue /Tank/Audio/Weapons/Tank_Cannon_Proj_Expl_Water_Cue
root=QualityLevel_0
nodes [
 {
  id=Attenuation_0
  children [Mixer_0]
  properties {attenuationSettings=/Game/Sounds/Attenuation/Fort_Explosion_Close_Athena_Att, attenuationOverrides {falloffDistance=1600}}
 }
 {
  id=Attenuation_1
  children [Modulator_0]
  properties {
   attenuationSettings=/Game/Sounds/Attenuation/Fort_Explosion_Medium_Athena_Att
   attenuationOverrides {falloffDistance=3000, attenuationShapeExtents {x=1000}}
  }
 }
 {
  id=Attenuation_2
  children [Mixer_2]
  properties {attenuationSettings=&Fort_Explosion_Close_Athena_Att, attenuationOverrides {falloffDistance=1600}}
 }
 {
  id=Attenuation_3
  children [Modulator_2]
  properties {
   attenuationSettings=&Fort_Explosion_Medium_Athena_Att
   attenuationOverrides {falloffDistance=3000, attenuationShapeExtents {x=1000}}
  }
 }
 {
  id=Attenuation_4
  children [Modulator_5]
  properties {
   attenuationSettings=/Game/Sounds/Attenuation/Fort_Explosion_Distant_Athena_Att
   attenuationOverrides {falloffDistance=3000, attenuationShapeExtents {x=1000}}
  }
 }
 {
  id=Attenuation_6
  children [Modulator_4]
  properties {
   attenuationSettings=&Fort_Explosion_Distant_Athena_Att
   attenuationOverrides {falloffDistance=3000, attenuationShapeExtents {x=1000}}
  }
 }
 {id=Mixer_0, children [Random_2,Random_3], properties {inputVolume [1,1]}}
 {id=Mixer_1, children [SoundClass_4,SoundClass_1,SoundClass_5], properties {inputVolume [1,1,1]}}
 {id=Mixer_2, children [Random_0,Random_6], properties {inputVolume [1,1]}}
 {id=Mixer_3, children [SoundClass_2,SoundClass_0,SoundClass_3], properties {inputVolume [1,1,1]}}
 {id=Modulator_0, children [Random_4], properties {pitchMax=1.1, pitchMin=0.9}}
 {id=Modulator_1, children [Attenuation_0]}
 {id=Modulator_2, children [Random_7], properties {pitchMax=1.1, pitchMin=0.9}}
 {id=Modulator_3, children [Attenuation_2]}
 {id=Modulator_4, children [Random_1], properties {pitchMax=1.1, pitchMin=0.9}}
 {id=Modulator_5, children [Random_5], properties {pitchMax=1.1, pitchMin=0.9}}
 {id=QualityLevel_0, children [Mixer_1,Mixer_1,Mixer_3,Mixer_3]}
 {id=Random_0, children [WavePlayer_0, WavePlayer_1, WavePlayer_4, WavePlayer_12], properties {weights [1,1,1,1]}}
 {id=Random_1, children [WavePlayer_37, WavePlayer_38, WavePlayer_39, WavePlayer_17], properties {weights [1,1,1,1]}}
 {id=Random_2, children [WavePlayer_2,WavePlayer_3], properties {weights [1,1]}}
 {id=Random_3, children [WavePlayer_7,WavePlayer_8,WavePlayer_9], properties {bShouldExcludeFromBranchCulling=true, weights [1,1,1]}}
 {id=Random_4, children [WavePlayer_11, WavePlayer_18, WavePlayer_19], properties {bShouldExcludeFromBranchCulling=true, weights [1,1,1]}}
 {id=Random_5, children [WavePlayer_21, WavePlayer_22], properties {weights [1,1]}}
 {
  id=Random_6
  children [WavePlayer_13, WavePlayer_14, WavePlayer_15, WavePlayer_16]
  properties {bShouldExcludeFromBranchCulling=true, weights [1,1,1,1]}
 }
 {
  id=Random_7
  children [WavePlayer_33, WavePlayer_34, WavePlayer_35, WavePlayer_36]
  properties {bShouldExcludeFromBranchCulling=true, weights [1,1,1,1]}
 }
 {
  id=SoundClass_0
  children [Attenuation_1]
  properties {soundClassOverride=&Fort_WeaponRanged_Fire_Distant}
 }
 {
  id=SoundClass_1
  children [Attenuation_3]
  properties {soundClassOverride=&Fort_WeaponRanged_Fire_Distant}
 }
 {id=SoundClass_2, children [Modulator_1], properties {soundClassOverride=&Fort_WeaponRanged_Fire_3P_PSM}}
 {
  id=SoundClass_3
  children [Attenuation_4]
  properties {soundClassOverride=&Fort_WeaponRanged_Fire_Distant}
 }
 {id=SoundClass_4, children [Modulator_3], properties {soundClassOverride=&Fort_WeaponRanged_Fire_3P_PSM}}
 {
  id=SoundClass_5
  children [Attenuation_6]
  properties {soundClassOverride=&Fort_WeaponRanged_Fire_Distant}
 }
 {id=WavePlayer_0, properties {soundWaveAssetPtr=/Game/Athena/Sounds/Weapons/Explosions/EXPL_Debris_Water_Wave_01}}
 {id=WavePlayer_1, properties {soundWaveAssetPtr=/Game/Athena/Sounds/Weapons/Explosions/EXPL_Debris_Water_Wave_02}}
 {id=WavePlayer_11, properties {soundWaveAssetPtr=&/WPN_Tank_Cannon_Proj_Expl_Med_01}}
 {id=WavePlayer_12, properties {soundWaveAssetPtr=/Game/Athena/Sounds/Weapons/Explosions/EXPL_Debris_Water_Wave_04}}
 {id=WavePlayer_13, properties {soundWaveAssetPtr=&/WPN_Tank_Cannon_Proj_Expl_Close_01}}
 {id=WavePlayer_14, properties {soundWaveAssetPtr=&/WPN_Tank_Cannon_Proj_Expl_Close_02}}
 {id=WavePlayer_15, properties {soundWaveAssetPtr=&/WPN_Tank_Cannon_Proj_Expl_Close_03}}
 {id=WavePlayer_16, properties {soundWaveAssetPtr=&/WPN_Tank_Cannon_Proj_Expl_Close_04}}
 {id=WavePlayer_17, properties {soundWaveAssetPtr=&/WPN_Tank_Cannon_Proj_Expl_Dist_03}}
 {id=WavePlayer_18, properties {soundWaveAssetPtr=&/WPN_Tank_Cannon_Proj_Expl_Med_02}}
 {id=WavePlayer_19, properties {soundWaveAssetPtr=&/WPN_Tank_Cannon_Proj_Expl_Med_03}}
 {id=WavePlayer_2, properties {soundWaveAssetPtr=&EXPL_Debris_Water_Wave_01}}
 {id=WavePlayer_21, properties {soundWaveAssetPtr=&/WPN_Tank_Cannon_Proj_Expl_Dist_04}}
 {id=WavePlayer_22, properties {soundWaveAssetPtr=&/WPN_Tank_Cannon_Proj_Expl_Dist_01}}
 {id=WavePlayer_3, properties {soundWaveAssetPtr=&EXPL_Debris_Water_Wave_02}}
 {id=WavePlayer_33, properties {soundWaveAssetPtr=&WPN_Tank_Cannon_Proj_Expl_Med_01}}
 {id=WavePlayer_34, properties {soundWaveAssetPtr=&WPN_Tank_Cannon_Proj_Expl_Med_02}}
 {id=WavePlayer_35, properties {soundWaveAssetPtr=&WPN_Tank_Cannon_Proj_Expl_Med_03}}
 {id=WavePlayer_36, properties {soundWaveAssetPtr=&/WPN_Tank_Cannon_Proj_Expl_Med_04}}
 {id=WavePlayer_37, properties {soundWaveAssetPtr=&WPN_Tank_Cannon_Proj_Expl_Dist_04}}
 {id=WavePlayer_38, properties {soundWaveAssetPtr=&WPN_Tank_Cannon_Proj_Expl_Dist_01}}
 {id=WavePlayer_39, properties {soundWaveAssetPtr=&/WPN_Tank_Cannon_Proj_Expl_Dist_02}}
 {id=WavePlayer_4, properties {soundWaveAssetPtr=/Game/Athena/Sounds/Weapons/Explosions/EXPL_Debris_Water_Wave_03}}
 {id=WavePlayer_7, properties {soundWaveAssetPtr=&WPN_Tank_Cannon_Proj_Expl_Close_01}}
 {id=WavePlayer_8, properties {soundWaveAssetPtr=&WPN_Tank_Cannon_Proj_Expl_Close_02}}
 {id=WavePlayer_9, properties {soundWaveAssetPtr=&WPN_Tank_Cannon_Proj_Expl_Close_03}}
]
properties {
 bOverrideConcurrency=true
 maxDistance=27500
 soundClassObject=&Fort_WeaponRanged_Fire_3P_PSM
 volumeMultiplier=0.8
 concurrencyOverrides {maxCount=3}
 concurrencySet [/Game/Sounds/Concurrency_Rules/Weapon_Explosion_Conc]
}
