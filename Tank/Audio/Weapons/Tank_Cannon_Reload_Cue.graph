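# Sound cue named for the tank cannon reload, with an empty node graph and no root, so it references no audio.
# It retains playback settings: a reduced volume, the Fort_Vehicles sound class and the shared Tank_Movement_Close_Att attenuation asset.
# The cue also stores reverb wet level values in its attenuation overrides.
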SoundCue /Tank/Audio/Weapons/Tank_Cannon_Reload_Cue
root=null
nodes []
properties {
 attenuationSettings=/Tank/Audio/Settings/Tank_Movement_Close_Att
 maxDistance=6024
 soundClassObject=&Fort_Vehicles
 volumeMultiplier=0.6
 attenuationOverrides {reverbWetLevelMax=0.7, reverbWetLevelMin=0.2}
}
