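# Sound cue that plays the looping aim-down-sights audio for the tank cannon's scope.
# A single looping wave player is the root and feeds the ADS loop wave straight out at reduced volume through the weapon foley sound class.
# The cue overrides attenuation with its own falloff, stereo spread and non-spatialized radius.
# Its concurrency override rejects new instances rather than stopping playing ones when the limit is reached.
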SoundCue /Tank/Audio/Weapons/Tank_Cannon_Scope_Loop_Cue
root=WavePlayer_2
nodes [{id=WavePlayer_2, properties {bLooping=true, soundWaveAssetPtr=&/WPN_Tank_Cannon_ADS_Loop_01}}]
properties {
 bOverrideAttenuation=true
 bOverrideConcurrency=true
 maxDistance=2000
 soundClassObject=&Fort_WeaponFoley
 volumeMultiplier=0.251
 attenuationOverrides {falloffDistance=1500, nonSpatializedRadiusStart=400, stereoSpread=300, attenuationShapeExtents {x=500}}
 concurrencyOverrides {resolutionRule=PreventNew}
}
