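# Sound cue for the tank cannon's scope zoom-out sound, filed under the weapon foley sound class.
# A single ADS-out wave plays through a Modulator left at default settings, with the cue's volume well reduced.
# Attenuation overrides turn off distance falloff and spatialization, so it plays as a flat 2D sound.
# A concurrency override blocks new instances once the limit is reached rather than cutting off playing ones.
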SoundCue /Tank/Audio/Weapons/Tank_Cannon_Scope_ZoomOut_Cue
root=Modulator_1
nodes [{id=Modulator_1, children [WavePlayer_1]}, {id=WavePlayer_1, properties {soundWaveAssetPtr=&/WPN_Tank_Cannon_ADS_Out_01}}]
properties {
 bOverrideAttenuation=true
 bOverrideConcurrency=true
 maxDistance=2097152
 soundClassObject=&Fort_WeaponFoley
 volumeMultiplier=0.251
 attenuationOverrides {bAttenuate=false, bSpatialize=false, falloffDistance=1500, nonSpatializedRadiusStart=200, attenuationShapeExtents {x=1200}}
 concurrencyOverrides {resolutionRule=PreventNew}
}
